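# Translucent unlit material for the Sprite Ghost's disc, usable on skeletal meshes, that draws a rotating swirl.
# A swirl mask on the second UV channel rotates over time with an adjustable offset, and its RGB channels feed named reroutes.
# Main, end-fade and swirl-edge colors blend through reroute-driven lerps and a final multiply.
# An inverted Fresnel multiplied by reroute values and a sharpness parameter drives a temporal-AA dither threshold.
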Material /SpecterGameplay/Assets/Sprite_Ghost/Materials/M_Sprite_Ghost_Disc
graph {
 properties {
  bAutomaticallySetUsageInEditor=false
  bUsedWithMeshDeformer=true
  bUsedWithSkeletalMesh=true
  blendMode=BLEND_Translucent
  shadingModel=MSM_Unlit
  usedShadingModels=Unlit
 }
 nodes [
  n1 = OneMinus(None=Fresnel(ExponentIn=ScalarParameter() {defaultValue=2, group=Fresnel, parameterName=Fresnel Exp, sortPriority=0}))
  n2 = Multiply(A=NamedRerouteUsage(), B=NamedRerouteUsage())
  MaterialFunctionCall(Alpha Threshold=Multiply(A=Multiply(A=@n1, B=@n2), B=ScalarParameter() {defaultValue=1, parameterName=Dither Sharpness})) {materialFunction=&DitherTemporalAA_Fortnite}
  n3 = VectorParameter() {group=Color, parameterName=Main Color, defaultValue=<rgba 0.3073 0.8707 1 1>}
  n4 = VectorParameter() {group=Color, parameterName=End Fade Color, defaultValue=<rgba 0 0.4618 0.5677 1>}
  n5 = NamedRerouteUsage()
  Multiply(
   A=LinearInterpolate(
    A=VectorParameter() {group=Color, parameterName=Swirl Edge Color, defaultValue=<rgba 0 0.1537 0.7379 1>}.RGB
    B=LinearInterpolate(A=@n3.RGB, B=@n4.RGB, Alpha=@n5)
    Alpha=NamedRerouteUsage()
   )
   B=NamedRerouteUsage()
  )
  n6 = Divide(A=Time(), B=ScalarParameter() {defaultValue=5, group=Spiral, parameterName=B})
  n7 = ScalarParameter() {group=Spiral, parameterName=Additional Spiral Rotation, sortPriority=0}
  n8 = TextureObjectParameter() {group=Spiral, parameterName=Swirl Mask, samplerType=SAMPLERTYPE_Masks, texture=/SpecterGameplay/Assets/Sprite_Ghost/Textures/T_Disc_Swirl}
  n9 = TextureSample(
   UVs=MaterialFunctionCall(UVs=TextureCoordinate() {coordinateIndex=1}, "Rotation Angle (0-1)"=Add(A=@n6, B=@n7)) {materialFunction=&CustomRotator}.Rotated Values
   Tex=@n8
  ) {samplerType=SAMPLERTYPE_Masks, texture=/SpecterGameplay/Assets/Sprite_Ghost/Textures/T_Sprite_Ghost}
  NamedRerouteDeclaration(None=@n9.R) {name=Spiral R, nodeColor=<rgba 1 0.7734 0 1>}
  NamedRerouteDeclaration(None=@n9.G) {name=Spiral G, nodeColor=<rgba 0.4531 0 1 1>}
  NamedRerouteDeclaration(None=@n9.B) {name=Spiral B, nodeColor=<rgba 0 1 0.1328 1>}
 ]
}
