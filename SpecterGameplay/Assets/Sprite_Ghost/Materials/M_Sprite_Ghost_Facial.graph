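# Subsurface material for the face of the Sprite Ghost character, covering eyes and mouth, with usage flags for skeletal, morph target, cloth and Niagara meshes.
# It overrides attributes from the shared CharacterParent PreFX function, then passes the result through the PostFX function.
# The FX mask's blue channel blends eye and shine color and specular, while red blends roughness.
# The blended color also drives emissive, and one parameter feeds both subsurface color and thickness.
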Material /SpecterGameplay/Assets/Sprite_Ghost/Materials/M_Sprite_Ghost_Facial
graph {
 properties {
  bAutomaticallySetUsageInEditor=false
  bUseMaterialAttributes=true
  bUsedWithClothing=true
  bUsedWithMeshDeformer=true
  bUsedWithMorphTargets=true
  bUsedWithNiagaraMeshParticles=true
  bUsedWithSkeletalMesh=true
  shadingModel=MSM_Subsurface
  usedShadingModels=Subsurface
 }
 nodes [
  n1 = MaterialFunctionCall() {materialFunction=&MF_CharacterParent_2023_PreFX}
  n2 = TextureSample() {samplerType=SAMPLERTYPE_Masks, texture=/SpecterGameplay/Assets/Sprite_Ghost/Textures/T_Sprite_Ghost_Body_FX}
  n3 = LinearInterpolate(
   A=VectorParameter() {parameterName=Eye Color, defaultValue=<rgba 0.0174 0.0001 0.0002 1>}.RGB
   B=VectorParameter() {parameterName=Shine Color, defaultValue=<rgba 0.8545 0.9271 0.91 1>}.RGB
   Alpha=@n2.B
  )
  n4 = VectorParameter() {parameterName=Subsurface Color, defaultValue=<rgba 0.0087 0.0112 0.0573 0.2522>}
  n5 = SetMaterialAttributes!(
   MaterialAttributes=GetMaterialAttributes(None=@n1.Result).MaterialAttributes
   Base Color=@n3
   Normal=TextureSample() {samplerType=SAMPLERTYPE_Normal, texture=/SpecterGameplay/Assets/Sprite_Ghost/Textures/T_Sprite_Ghost_Body_N}.RGB
   Roughness=LinearInterpolate(
    A=ScalarParameter() {defaultValue=1, parameterName=Roughness Eyes}
    B=ScalarParameter() {defaultValue=0.15, parameterName=Roughness Mouth}
    Alpha=@n2.R
   )
   Specular=LinearInterpolate(A=ScalarParameter() {defaultValue=0.5, parameterName=Spec Mouth}, B=ScalarParameter() {defaultValue=1, parameterName=Spec Eyes}, Alpha=@n2.B)
   Subsurface Color=@n4.RGB
   Surface Thickness=@n4.RGB
   Emissive Color=Multiply(A=@n3, B=ScalarParameter() {defaultValue=0.25, parameterName=Emissive Intensity})
  ) {attributeSetTypes [
    #1
    #2
    #3
    #4
    #5
    #6
    #7
   ]}
  MaterialFunctionCall(Material=@n5, M=@n1.M, Pre Skinned Local Position=@n1.M, Distance Blend=@n1.M, Fresnel=@n1.M) {materialFunction=&MF_CharacterParent_2023_PostFX}
 ]
}
