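# Material function for the ghost sprite that computes an animated wobble, apparently for World Position Offset.
# Two waves scroll along world Z over time, the second at 1.3 times Speed with a shorter wavelength.
# Their nested sine/cosine sum is scaled by Intensity and masked by red vertex color.
# A Velocity X/Y offset branch is built but never reaches the output.
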MaterialFunction /SpecterGameplay/Assets/Sprite_Ghost/Materials/MaterialFunctions/MF_Sprite_Ghost
graph {nodes [
  n1 = WorldPosition()
  n2 = Time()
  n3 = ScalarParameter() {defaultValue=40, group=WPO, parameterName=Speed}
  n4 = Add(A=@n1.Z, B=Multiply(A=@n2, B=@n3) {constB=40})
  n5 = Power(Base=Sine(None=Divide(A=@n4) {constB=200}))
  n6 = Multiply(A=@n3) {constB=1.3}
  n7 = Divide(A=Add(A=@n1.Z, B=Multiply(A=@n2, B=@n6) {constB=36})) {constB=150}
  n8 = Multiply(A=Add(A=Cosine(None=@n5), B=Sine(None=@n7)), B=ScalarParameter() {defaultValue=5, group=WPO, parameterName=Intensity}) {constB=4}
  n9 = ScalarParameter() {group=WPO, parameterName=Velocity X, sliderMax=1, sliderMin=-1}
  n10 = ScalarParameter() {group=WPO, parameterName=Velocity Y, sliderMax=1, sliderMin=-1}
  Add(
   A=@n8
   B=MaterialFunctionCall(R=Multiply(A=@n9), G=Multiply(A=@n10), B=Constant()) {materialFunction=&AppendMany}.RGB
  )
  FunctionOutput(None=Multiply(A=VertexColor().R, B=@n8))
 ]}
