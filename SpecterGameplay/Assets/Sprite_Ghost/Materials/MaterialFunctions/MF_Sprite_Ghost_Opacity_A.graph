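# Material function that builds opacity and color masks for the ghost sprite, with animated holes, apparently for its tail.
# A tiled, panning dissolve texture is inverted, weighted by axis-aligned Fresnel, then sharpened into a mask by subtract, divide and saturate.
# A second Fresnel fades that mask at grazing edges before inverted vertex red adds solid areas.
# Vertex-red falloffs scale the result into Color Mask and Opacity Mask; Tex Out exposes the inverted texture.
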MaterialFunction /SpecterGameplay/Assets/Sprite_Ghost/Materials/MaterialFunctions/MF_Sprite_Ghost_Opacity_A
graph {nodes [
  n1 = VertexColor()
  n2 = OneMinus(None=@n1.R)
  n3 = Constant() {r=2}
  n4 = Multiply(
   A=TextureCoordinate()
   B=AppendVector(
    A=ScalarParameter() {defaultValue=6, group=Opacity UV, parameterName=Tile U}
    B=ScalarParameter() {defaultValue=1.5, group=Opacity UV, parameterName=Tile V}
   )
  )
  n5 = AppendVector(
   A=ScalarParameter() {defaultValue=-0.2, group=Opacity UV, parameterName=Speed X}
   B=ScalarParameter() {defaultValue=-0.3, group=Opacity UV, parameterName=Speed Y}
  )
  n6 = OneMinus(None=TextureSample(
    UVs=Panner(Coordinate=@n4, Speed=@n5)
    Tex=TextureObjectParameter() {group=Opacity, parameterName=Tail Holes Texture, texture=/SpecterGameplay/Assets/Sprite_Ghost/Textures/T_Dissolve_A}
   ).R)
  n7 = Multiply(
   A=Add(A=MaterialFunctionCall(Exponent=@n3) {materialFunction=&AxisAlignedFresnel}.Fresnel)
   B=Add(A=Constant() {desc=Tweaking texture brightness, r=0.25}, B=@n6) {constB=0.25}
  )
  n8 = ScalarParameter() {defaultValue=1.2, desc=0.6, group=Opacity, parameterName=Texture Subtract}
  n9 = Saturate(None=Divide(A=Subtract(A=@n7, B=@n8), B=ScalarParameter() {defaultValue=0.05, group=Opacity, parameterName=Texture Divide}))
  n10 = Saturate(None=Fresnel() {baseReflectFraction=0.0001, exponent=2})
  n11 = Multiply(A=@n9, B=Multiply(A=Subtract(A=@n10) {constB=0.1}) {constB=5})
  n12 = Add(A=Power(Base=@n2), B=Saturate(None=Subtract(A=@n9, B=@n11)))
  FunctionOutput(None=Saturate(None=Multiply(A=@n2, B=@n12))) {outputName=Color Mask}
  FunctionOutput(None=@n6) {outputName=Tex Out, sortPriority=10}
  n13 = Saturate(None=OneMinus(None=Power(Base=@n1.R) {constExponent=10}))
  FunctionOutput(None=Saturate(None=Multiply(A=@n13, B=@n12))) {outputName=Opacity Mask}
 ]}
