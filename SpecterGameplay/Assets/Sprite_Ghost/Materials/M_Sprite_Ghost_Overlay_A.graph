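# Translucent unlit overlay material for the Sprite Ghost skeletal mesh, giving it a glowing, bubbly body with colored edges.
# Fresnel, vertex color and optional bubbles (two animated parallax layers) blend core to edge color, with an outer rim added.
# Two scene-depth fades blend in spiral emissive and an inner-object color.
# Side branches build a body gradient, dithered distance fade and bone mask from custom HLSL multi-tap scene-texture blurs.
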Material /SpecterGameplay/Assets/Sprite_Ghost/Materials/M_Sprite_Ghost_Overlay_A
graph {
 properties {
  bAutomaticallySetUsageInEditor=false
  bUsedWithMeshDeformer=true
  bUsedWithSkeletalMesh=true
  blendMode=BLEND_Translucent
  shadingModel=MSM_Unlit
  usedShadingModels=Unlit
 }
 nodes [
  n1 = VectorParameter() {group=_Color, parameterName=Edge Color, defaultValue=<rgba 0.7016 0.5489 0.8212 1>}
  n2 = ScalarParameter() {defaultValue=0.5096, parameterName=Outer Fresnel Intensity}
  n3 = Fresnel(ExponentIn=ScalarParameter() {defaultValue=5, parameterName=Outer Fresnel Exp})
  n4 = MaterialFunctionCall() {materialFunction=/SpecterGameplay/Assets/Sprite_Ghost/Materials/MaterialFunctions/MF_Sprite_Ghost_Opacity_A}
  n5 = Fresnel() {exponent=2}
  n6 = Add(A=VertexColor().R, B=@n5)
  n7 = Add(A=NamedRerouteUsage(), B=@n6)
  n8 = LinearInterpolate(
   A=Multiply(A=NamedRerouteUsage(), B=VectorParameter() {group=_Color, parameterName=Core Color, defaultValue=<rgba 0.1443 0.0986 0.2396 1>}.RGB)
   B=@n1.RGB
   Alpha=StaticSwitchParameter(True=Divide(A=@n7) {constB=3}, False=Divide(A=@n6)) {defaultValue=true, group=Bubble, parameterName=Use Bubbles, sortPriority=0}
  )
  n9 = Multiply(A=NamedRerouteUsage(), B=ScalarParameter() {defaultValue=3, group=_Color, parameterName=Spiral Emissive Intensity}) {constB=2}
  n10 = SceneTexture() {sceneTextureId=PPI_SceneDepth}
  n11 = SceneDepth()
  n12 = ShadingPathSwitch(Default=@n10.Color, Deferred=@n10.Color, Forward=@n11.R, Mobile=@n11.R)
  n13 = NamedRerouteDeclaration(None=Subtract(A=ComponentMask(None=@n12) {r=true}, B=NamedRerouteUsage())) {name=CameraFade, nodeColor=<rgba 0.9453 1 0 1>}
  n14 = Saturate(None=Divide(A=@n13, B=ScalarParameter() {defaultValue=140, group=Blur, parameterName=DistFade}) {constB=20})
  n15 = ScalarParameter() {defaultValue=2, group=Blur, parameterName=ScenDepthPower}
  n16 = Saturate(None=Multiply(A=Power(Base=@n14, Exp=@n15), B=ScalarParameter() {defaultValue=2, group=Blur, parameterName=ScenDepthIntensity}))
  n17 = Multiply(A=OneMinus(None=@n16), B=NamedRerouteUsage())
  n18 = Saturate(None=Divide(A=@n13, B=ScalarParameter() {defaultValue=66.4719, group=Blur, parameterName=DistFade2}) {constB=20})
  n19 = ScalarParameter() {defaultValue=4.2382, group=Blur, parameterName=ScenDepthPower2}
  n20 = Saturate(None=Multiply(A=Power(Base=@n18, Exp=@n19), B=ScalarParameter() {defaultValue=1.3448, group=Blur, parameterName=ScenDepthIntensity2}))
  Add(
   A=Multiply(A=Multiply(A=@n1.RGB, B=@n2), B=Multiply(A=@n3, B=@n4.Color Mask))
   B=LinearInterpolate(
    A=LinearInterpolate(A=@n8, B=@n9, Alpha=@n17)
    B=VectorParameter() {group=_Color, parameterName=Inner Object Color, defaultValue=<rgba 0.0782 0.7379 0.8796 1>}.RGB
    Alpha=OneMinus(None=@n20)
   )
  )
  MaterialExpressionNamedRerouteUsage_2 = NamedRerouteUsage()
  MaterialExpressionConstant_3 = Constant() {r=1}
  n21 = Custom!(stepSize=@MaterialExpressionNamedRerouteUsage_2, iterations=@MaterialExpressionConstant_3) {
   code="float dist = stepSize / iterations;\r\nint BufferIndex = 2;\r\nfloat2 offsets[4] = { float2(1,0), float2(-1,0), float2(0,1), float2(0,-1)};\r\nfloat2 UV = GetDefaultSceneTextureUV(Parameters, int(BufferIndex));\r\nfloat2 NewUV = UV;\r\nfloat3 CurColor=0;\r\n\r\nfor (int i = 0; i < (int)iterations; i++)\r\n{\r\n\tfor (int j = 0; j < 4; j++)\r\n\t{\r\n\t\tNewUV = UV + offsets[j] * (i+1) * dist;\r\n\t\tCurColor += SceneTextureLookupCommon(Parameters,NewUV, BufferIndex, false);\r\n\t}\r\n}\r\n\r\nCurColor /= (iterations * 4.0);\r\nreturn CurColor;\r\n"
   inputs [
    {
     inputName=stepSize
     input {expression=&self:MaterialExpressionNamedRerouteUsage_2}
    }
    {
     inputName=iterations
     input {expression=&self:MaterialExpressionConstant_3}
    }
   ]
  }
  n22 = Saturate(None=Desaturation(None=@n21, Fraction=Constant() {r=1}))
  Add(A=Multiply(A=@n8, B=NamedRerouteUsage()), B=NamedRerouteDeclaration(None=OneMinus(None=@n22)) {name=BoneArea, nodeColor=<rgba 1 0 0.2578 1>})
  n23 = TextureSampleParameter2D() {parameterName=Body FX, samplerType=SAMPLERTYPE_Masks, texture=/SpecterGameplay/Assets/Sprite_Ghost/Textures/T_Sprite_Ghost_Body_FX}
  n24 = NamedRerouteDeclaration(None=@n23.RGBA) {name=Body FX, nodeColor=<rgba 1 0 0.1406 1>}
  ComponentMask(None=@n24) {r=true}
  Custom(stepSize=@MaterialExpressionNamedRerouteUsage_2, iterations=@MaterialExpressionConstant_3) {
   code="float dist = stepSize / iterations;\r\nint BufferIndex = 2;\r\nfloat2 offsets[8] = { float2(1,0), float2(-1,0), float2(0,1), float2(0,-1),float2(.5,.5), float2(-.5,-.5), float2(-.5,.5), float2(.5,-.5)};\r\nfloat2 UV = GetDefaultSceneTextureUV(Parameters, int(BufferIndex));\r\nfloat2 NewUV = UV;\r\nfloat3 CurColor=0;\r\n\r\nfor (int i = 0; i < (int)iterations; i++)\r\n{\r\n\tfor (int j = 0; j < 8; j++)\r\n\t{\r\n\t\tNewUV = UV + offsets[j] * (i+1) * dist;\r\n\t\tCurColor += SceneTextureLookupCommon(Parameters,NewUV, BufferIndex, false);\r\n\t}\r\n}\r\n\r\nCurColor /= (iterations * 8.0);\r\nreturn CurColor;\r\n"
   inputs [
    {
     inputName=stepSize
     input {expression=&self:MaterialExpressionNamedRerouteUsage_2}
    }
    {
     inputName=iterations
     input {expression=&self:MaterialExpressionConstant_3}
    }
   ]
  }
  MaterialFunctionCall() {materialFunction=/SpecterGameplay/Assets/Sprite_Ghost/Materials/MaterialFunctions/MF_Sprite_Ghost}
  n25 = WorldPosition()
  n26 = Subtract(
   A=ComponentMask(None=TransformPosition(None=@n25.XYZ) {transformSourceType=TRANSFORMPOSSOURCE_World, transformType=TRANSFORMPOSSOURCE_View}) {b=true}
   B=ScalarParameter() {defaultValue=40, group=Blur, parameterName=BlurDistanceBias}
  )
  n27 = Saturate(None=Divide(A=@n26, B=ScalarParameter() {defaultValue=100, group=Blur, parameterName=BlurDistanceFade}))
  NamedRerouteDeclaration(None=LinearInterpolate(
    A=@n27
    B=MaterialFunctionCall(Alpha Threshold=@n27) {materialFunction=&DitherTemporalAA}.Result
    Alpha=ScalarParameter() {defaultValue=0.1, group=Blur, parameterName=BlurDitherMulitplier}
   )) {name=BlurDistance, nodeColor=<rgba 0 1 0.9766 1>}
  n28 = OneMinus(None=@n5)
  NamedRerouteDeclaration(None=Saturate(None=Power(Base=@n28) {constExponent=10})) {name=Bubble Fresnel, nodeColor=<rgba 0.0312 1 0 1>}
  NamedRerouteDeclaration(None=@n16) {name=SceneDepthAlpha, nodeColor=<rgba 0 0.9297 1 1>}
  n29 = VectorParameter() {group=_Color, parameterName=Middle Gradient, defaultValue=<rgba 0.855 0.107 0.7835 1>}
  n30 = VectorParameter() {group=_Color, parameterName=Bottom Gradient, defaultValue=<rgba 0.7016 0.5489 0.8212 1>}
  n31 = ComponentMask(None=@n24) {b=true}
  NamedRerouteDeclaration(None=LinearInterpolate(
    A=VectorParameter() {group=_Color, parameterName=Top Gradient, defaultValue=<rgba 0.0782 0.7379 0.8796 1>}.RGB
    B=LinearInterpolate(A=@n29.RGB, B=@n30.RGB, Alpha=@n31)
    Alpha=ComponentMask(None=@n24) {g=true}
   )) {name=Body Color Gradient, nodeColor=<rgba 0.7812 1 0 1>}
  n32 = VertexColor()
  n33 = Divide(A=Time()) {constB=100}
  n34 = Add(A=TextureCoordinate(), B=Divide(A=Sine(None=@n33)) {constB=10}) {constB=0.3}
  n35 = Multiply(A=ScalarParameter() {defaultValue=-2, group=Bubble, parameterName=Bubble UV Depth}) {constB=0.7}
  n36 = ScalarParameter() {defaultValue=5, group=Bubble, parameterName=Bubble UV Tiling}
  n37 = TextureObjectParameter() {group=Bubble, parameterName=BubbleTexture, sortPriority=1, texture=/SpecterGameplay/Assets/Sprite_Ghost/Textures/T_Bubbles_A}
  n38 = TextureSample(UVs=Multiply(A=BumpOffset(Coordinate=@n34, Height=@n35), B=@n36) {constB=5}, Tex=@n37)
  n39 = Add(A=TextureCoordinate(), B=Divide(A=Time()) {constB=120})
  n40 = ScalarParameter() {defaultValue=-2, group=Bubble, parameterName=Bubble UV Depth}
  n41 = TextureSample(UVs=Multiply(A=@n36, B=BumpOffset(Coordinate=@n39, Height=@n40)) {constB=5}, Tex=@n37)
  n42 = Add(A=OneMinus(None=ScalarParameter() {defaultValue=0.8848, group=Bubble, parameterName=Bubble Intensity, sliderMax=1})) {constB=0.1}
  n43 = Saturate(None=Divide(A=Add(A=@n38.RGB, B=@n41.RGB), B=Multiply(A=@n42) {constB=10}) {constB=1})
  n44 = NamedRerouteUsage()
  NamedRerouteDeclaration(None=Multiply(A=OneMinus(None=@n32.R), B=Multiply(A=@n43, B=@n44))) {name=Bubble Mask, nodeColor=<rgba 1 0 0.3516 1>}
  NamedRerouteDeclaration(None=@n26) {name=CameraDist, nodeColor=<rgba 0 1 0.5781 1>}
  n45 = TextureSampleParameter2D() {parameterName=Body N, samplerType=SAMPLERTYPE_Normal, texture=/SpecterGameplay/Assets/Sprite_Ghost/Textures/T_Sprite_Ghost_Body_N}
  NamedRerouteDeclaration(None=@n45.RGB) {name=Body N, nodeColor=<rgba 0 0.4609 1 1>}
 ]
}
