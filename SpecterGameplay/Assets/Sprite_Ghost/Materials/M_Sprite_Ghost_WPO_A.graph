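# Masked, unlit, two-sided material for the Sprite Ghost, with an isWisp? parameter blending in a wisp variant.
# WPO adds MF_Sprite_Ghost motion to a clamped, UV-weighted pull driven by Niagara dynamic parameters.
# Ghost emissive layers Fresnel and vertex-color gradients, spiral glow, scrolling bubbles and a fake sky-light specular.
# Wisps use two-color, time-of-day-scaled emissive and camera-faded alpha; opacity masking uses temporal and Bayer dithering.
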Material /SpecterGameplay/Assets/Sprite_Ghost/Materials/M_Sprite_Ghost_WPO_A
graph {
 properties {
  bAutomaticallySetUsageInEditor=false
  bCastRayTracedShadows=false
  bUsedWithMeshDeformer=true
  bUsedWithSkeletalMesh=true
  blendMode=BLEND_Masked
  previewMesh=/SpecterGameplay/Gameplay/Effects/Haunt/SM_Haunt_WispMesh
  shadingModel=MSM_Unlit
  twoSided=true
  usedShadingModels=Unlit
 }
 nodes [
  n1 = DynamicParameter() {desc=Parameters set by Niagara, parameterIndex=3, defaultValue=<rgba 0 1 1 1>, paramNames [X,Y,Z,Intensity]}
  n2 = MaterialFunctionCall(Float3=Subtract(
    A=MaterialFunctionCall(X=@n1.X, Y=@n1.X, Z=@n1.X) {materialFunction=&MakeFloat3}.Result
    B=ParticlePositionWS()
   )) {materialFunction=&BreakOutFloat3Components}
  n3 = ScalarParameter() {defaultValue=5, desc=Interpolate Clamp XY, group=WPO, parameterName=Interpolate XY Clamp, sortPriority=3}
  n4 = ScalarParameter() {defaultValue=3, desc=Interpolate Clamp Z, group=WPO, parameterName=Interpolate Clamp Z, sortPriority=4}
  n5 = AppendVector(
   A=Clamp(None=AppendVector(A=@n2.R, B=@n2.R), Min=Multiply(A=@n3) {constB=-1}, Max=@n3) {maxDefault=1000, minDefault=-1000}
   B=Clamp(None=@n2.B, Min=Multiply(A=@n4) {constB=-1}, Max=@n4) {minDefault=-1}
  )
  n6 = TextureCoordinate()
  n7 = Multiply(
   A=@n1.Intensity
   B=Power(
    Base=ComponentMask(None=@n6) {g=true}
    Exp=ScalarParameter() {defaultValue=3, desc=Interpolate Power, group=WPO, parameterName=Interpolate Power, sortPriority=2}
   )
  )
  Add(
   A=MaterialFunctionCall() {materialFunction=/SpecterGameplay/Assets/Sprite_Ghost/Materials/MaterialFunctions/MF_Sprite_Ghost}.Result
   B=LinearInterpolate(B=LinearInterpolate(B=@n5, Alpha=@n7), Alpha=ScalarParameter() {group=AssetVariant, parameterName=isWisp?})
  )
  n8 = TextureCoordinate()
  n9 = ComponentMask(None=@n8) {g=true}
  n10 = Time()
  n11 = Divide(A=Sine(None=Divide(A=@n10) {constB=150})) {constB=10}
  n12 = BumpOffset(
   Coordinate=AppendVector(A=ComponentMask(None=@n8) {r=true}, B=Subtract(A=@n9, B=@n11))
   Height=ScalarParameter() {defaultValue=-5, group=Bubble, parameterName=Bubble UV Depth}
  )
  n13 = ScalarParameter() {defaultValue=10, group=Bubble, parameterName=Bubble UV Tiling}
  n14 = TextureObjectParameter() {group=Bubble, parameterName=BubbleTexture, sortPriority=1, texture=/SpecterGameplay/Assets/Sprite_Ghost/Textures/T_Bubbles_A}
  n15 = Multiply(A=@n13) {constB=2}
  n16 = TextureCoordinate()
  n17 = Time()
  n18 = AppendVector(A=ComponentMask(None=@n16) {r=true}, B=Subtract(A=ComponentMask(None=@n16) {g=true}, B=Divide(A=@n17) {constB=200}))
  n19 = Multiply(A=TextureSample(UVs=Multiply(A=@n12, B=@n13) {constB=5}, Tex=@n14).R, B=TextureSample(UVs=Multiply(A=@n15, B=@n18) {constB=5}, Tex=@n14).R)
  n20 = OneMinus(None=ScalarParameter() {defaultValue=4, group=Bubble, parameterName=Bubble Intensity, sliderMax=1})
  Divide(A=@n19, B=Multiply(A=Add(A=@n20) {constB=0.1}) {constB=10}) {constB=1}
  n21 = MaterialFunctionCall() {materialFunction=/SpecterGameplay/Assets/Sprite_Ghost/Materials/MaterialFunctions/MF_Sprite_Ghost_Opacity_A}
  n22 = NamedRerouteUsage()
  n23 = Add(A=@n21.Opacity Mask, B=Multiply(A=@n21.Color Mask, B=@n22))
  n24 = Divide(A=@n22) {constB=1.5}
  n25 = MaterialFunctionCall(Alpha Threshold=LinearInterpolate(A=NamedRerouteUsage(), B=Subtract(A=@n23, B=@n24), Alpha=@n22)) {materialFunction=&DitherTemporalAA}
  MaterialFunctionCall!(
   In=@n25.Result
   Iterations=ScalarParameter() {defaultValue=8, group=Opacity, parameterName=Bayer Iterations}
   Scale=ScalarParameter() {defaultValue=6, group=Opacity, parameterName=Bayer Scale}
  ) {materialFunction=/Game/Packages/Fortress_SharedMaterials/Base_Material_Functions/MF_DitherBayerNoTex}
  n26 = Add(
   A=SkyAtmosphereLightDirection()
   B=Multiply(A=MaterialFunctionCall() {materialFunction=&CameraDirectionVector}.Result) {constB=-1}
  )
  n27 = Saturate(None=DotProduct(A=PixelNormalWS(), B=Normalize(VectorInput=@n26)))
  n28 = Multiply(A=Fresnel() {baseReflectFraction=0, exponent=2}, B=Power(Base=@n27) {constExponent=4})
  n29 = Divide(A=Power(Base=@n27) {constExponent=100}) {constB=1}
  n30 = ShadowReplace(Default=Saturate(None=Add(A=@n28, B=@n29)), Shadow=Constant() {r=1})
  MaterialFunctionCall(Alpha Threshold=@n30) {materialFunction=&DitherTemporalAA}
  n31 = NamedRerouteDeclaration(None=Fresnel() {baseReflectFraction=0, exponent=1}) {name=Core Mask, nodeColor=<rgba 0.3594 1 0 1>}
  n32 = Power(Base=@n31) {constExponent=5}
  n33 = Multiply(A=NamedRerouteDeclaration(None=VertexColor().R) {name=Bottom Mask, nodeColor=<rgba 1 0.4453 0 1>}) {constB=2}
  n34 = LinearInterpolate(A=NamedRerouteUsage(), B=NamedRerouteUsage(), Alpha=Saturate(None=Add(A=@n32, B=@n33)))
  NamedRerouteDeclaration(None=LinearInterpolate(A=@n34, B=Add(A=@n34) {constB=0.2}, Alpha=Fresnel() {baseReflectFraction=0, exponent=6})) {name=Color Combination, nodeColor=<rgba 1 0 0.2109 1>}
  n35 = NamedRerouteUsage()
  n36 = Power(Base=Add(A=@n31, B=Multiply(A=@n35) {constB=2}))
  NamedRerouteDeclaration(None=LinearInterpolate(
    A=VectorParameter() {group=Color, parameterName=Core Color, sortPriority=1, defaultValue=<rgba 0.1912 0.1248 0.7011 1>}.RGB
    B=NamedRerouteUsage()
    Alpha=Saturate(None=@n36)
   )) {name=First Color Combination, nodeColor=<rgba 1 0.3516 0 1>}
  n37 = LocalPosition() {includedOffsets=ExcludeOffsets, localOrigin=InstancePreSkinning}
  n38 = Divide(A=Add(A=@n37.Z) {constB=-12}) {constB=20}
  n39 = Divide(A=Add(A=OneMinus(None=@n37.Z)) {constB=35}) {constB=20}
  n40 = Multiply(A=VertexInterpolator(VS=Multiply(A=@n38, B=@n39)).PS) {constB=2}
  NamedRerouteDeclaration(None=LinearInterpolate(
    A=NamedRerouteUsage()
    B=VectorParameter() {group=Color, parameterName=Spiral Glow, defaultValue=<rgba 0.0299 0.3717 0.4045 1>}.RGB
    Alpha=Saturate(None=@n40)
   )) {name=Face Glow, nodeColor=<rgba 0.2891 1 0 1>}
  n41 = VertexColor()
  n42 = Multiply(A=Multiply(A=OneMinus(None=@n41.R), B=Saturate(None=@n19)), B=OneMinus(None=NamedRerouteUsage()))
  NamedRerouteDeclaration(None=Saturate(None=Multiply(A=@n42) {constB=5})) {name=Bubble Mask, nodeColor=<rgba 1 0 0.3516 1>}
  n43 = NamedRerouteUsage()
  n44 = MaterialFunctionCall() {materialFunction=/Game/Packages/Fortress_SharedMaterials/Base_Material_Functions/SkylightColorAndIntensity.SkyLightColorAndIntensity}
  n45 = LinearInterpolate(A=Multiply(A=Multiply(A=@n43, B=@n44.Color)) {constB=0.5}, B=NamedRerouteUsage(), Alpha=NamedRerouteUsage())
  n46 = Add(A=Power(Base=NamedRerouteUsage()), B=@n45)
  n47 = NamedRerouteUsage()
  NamedRerouteDeclaration(None=Multiply(A=LinearInterpolate(A=@n46, B=@n45, Alpha=@n47), B=ScalarParameter() {defaultValue=2, group=Color, parameterName=Emissive Intensity}) {constB=2}) {name=Emissive, nodeColor=<rgba 0 1 0.5547 1>}
  n48 = NamedRerouteUsage()
  n49 = Add(A=LinearInterpolate(A=@n48, B=NamedRerouteUsage(), Alpha=NamedRerouteUsage()), B=NamedRerouteUsage())
  NamedRerouteDeclaration(None=Saturate(None=Add(A=@n48, B=@n49))) {name=Opacity, nodeColor=<rgba 1 0.2344 0 1>}
  NamedRerouteDeclaration(None=@n30) {name=Specular Mask, nodeColor=<rgba 0 0.2266 1 1>}
  n50 = MaterialFunctionCall(Color=VertexInterpolator(VS=LocalPosition() {localOrigin=InstancePreSkinning}.XYZ).PS) {materialFunction=&SplitComponents}
  n51 = Divide(A=@n50.R) {constB=20}
  n52 = Divide(A=OneMinus(None=@n50.R)) {constB=20}
  NamedRerouteDeclaration(None=Add(A=Power(Base=@n51) {constExponent=3}, B=Power(Base=@n52) {constExponent=3})) {name=Sides Mask, nodeColor=<rgba 1 0 0.8672 1>}
  n53 = TwoSidedSign()
  NamedRerouteDeclaration(None=Saturate(None=@n53)) {name=Frontface Mask, nodeColor=<rgba 0.1719 0 1 1>}
  NamedRerouteDeclaration(None=Saturate(None=Multiply(A=@n53) {constB=-1})) {name=Backface Mask, nodeColor=<rgba 0.1484 1 0 1>}
  NamedRerouteDeclaration(None=@n50.B) {name=Local Position Z, nodeColor=<rgba 0 1 0.8125 1>}
  n54 = NamedRerouteDeclaration(None=VectorParameter() {group=Color, parameterName=WispColor01, defaultValue=<rgba 0.7 0 0.5 1>}.RGB) {name=Color 1, nodeColor=<rgba 1 0 0.5625 1>}
  n55 = ScalarParameter() {defaultValue=1}
  n56 = VertexColor()
  n57 = OneMinus(None=MaterialFunctionCall() {materialFunction=&MF_Sprite_Ghost_Opacity_A}.Tex Out)
  n58 = Constant() {r=3}
  n59 = Saturate(None=Add(A=Multiply(A=@n57, B=@n58), B=Constant() {r=-0.2}))
  n60 = Constant() {desc=AlphaEdgeGlowPow, r=0.65}
  n61 = Add(
   A=Fresnel(
    ExponentIn=ScalarParameter() {defaultValue=0.5, parameterName=ColorFresnel_Exp}
    BaseReflectFractionIn=ScalarParameter() {defaultValue=0.04, parameterName=ColorFresnel_BaseReflectFrac}
   ) {baseReflectFraction=0.0001, exponent=4}
   B=Multiply(A=VertexColor().R, B=Power(Base=@n59, Exp=@n60))
  )
  n62 = LinearInterpolate(
   A=NamedRerouteDeclaration(None=VectorParameter() {group=Color, parameterName=WispColor02, defaultValue=<rgba 0 0.5 1 1>}.RGB) {name=Color 2, nodeColor=<rgba 0 0.8828 1 1>}
   B=LinearInterpolate(A=@n54, B=LinearInterpolate(A=@n54, Alpha=@n55), Alpha=@n56.R)
   Alpha=Saturate(None=Add(A=@n56.R, B=@n61))
  )
  n63 = Multiply(A=NamedRerouteUsage(), B=ScalarParameter() {defaultValue=1, group=Color, parameterName=WispsSpecularBrightness}) {constB=0.65}
  n64 = ScalarParameter() {defaultValue=1, group=Lighting Adjustments, parameterName=TOD_Morning}
  n65 = ScalarParameter() {defaultValue=1, group=Lighting Adjustments, parameterName=TOD_Day}
  n66 = ScalarParameter() {defaultValue=0.85, group=Lighting Adjustments, parameterName=TOD_Evening}
  n67 = ScalarParameter() {defaultValue=0.6, group=Lighting Adjustments, parameterName=TOD_Night}
  NamedRerouteDeclaration(None=Multiply(
    A=Add(A=@n62, B=@n63)
    B=MaterialFunctionCall(Value_Morning=@n64, Value_Day=@n65, Value_Evening=@n66, Value_Night=@n67) {materialFunction=&Scalar_Time_of_Day}.Result
   )) {name=Wisps_Emissive, nodeColor=<rgba 0 0.4664 1 1>}
  Reroute(None=@n25.Result)
  n68 = OneMinus(None=VertexColor().R)
  n69 = Multiply(
   A=Power(
    Base=MaterialFunctionCall() {materialFunction=&MF_Sprite_Ghost_Opacity_A}.Color Mask
    Exp=Constant() {desc=isWisps needs more alpha spread, r=1.5}
   )
   B=Saturate(None=LinearInterpolate(Alpha=@n68) {constA=-0.5})
  )
  n70 = MaterialFunctionCall(
   Fade Length=ScalarParameter() {defaultValue=50, parameterName=CamFade_FadeLength}
   Fade Offset=ScalarParameter() {defaultValue=60, parameterName=CamFade_DepthOffset}
  ) {materialFunction=&CameraDepthFade}
  Reroute(None=Saturate(None=Multiply(A=@n69, B=@n70.Result))) {desc="isWisp = Translucent"}
  n71 = NamedRerouteUsage()
  n72 = NamedRerouteUsage()
  n73 = ScalarParameter() {group=AssetVariant, parameterName=isWisp?}
  Reroute(None=EyeAdaptationInverse(
    LightValueInput=LinearInterpolate(A=@n71, B=@n72, Alpha=@n73)
    AlphaInput=ScalarParameter() {defaultValue=1, group=Lighting Adjustments, parameterName=EyeAdaptionInverse}
   ).EyeAdaptationInverse)
 ]
}
