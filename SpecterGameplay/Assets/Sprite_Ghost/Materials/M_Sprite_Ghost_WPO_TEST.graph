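# Masked unlit test material for the Sprite Ghost on skeletal and deformer-driven meshes, with an isWisp scalar selecting the wisp variant.
# Color blends two tints by vertex red plus Fresnel and wisp-only alpha-edge glow, then gets time-of-day scaling and inverse eye adaptation.
# Opacity comes from a shared ghost function, with a wisp-dependent power and a temporal-dither branch.
# Despite the WPO name, the Pixel Offset parameter feeds nothing visible.
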Material /SpecterGameplay/Assets/Sprite_Ghost/Materials/M_Sprite_Ghost_WPO_TEST
graph {
 properties {
  bAutomaticallySetUsageInEditor=false
  bUsedWithMeshDeformer=true
  bUsedWithSkeletalMesh=true
  blendMode=BLEND_Masked
  shadingModel=MSM_Unlit
  usedShadingModels=Unlit
 }
 nodes [
  MaterialFunctionCall() {materialFunction=/SpecterGameplay/Assets/Sprite_Ghost/Materials/MaterialFunctions/MF_Sprite_Ghost}
  n1 = Constant() {r=1}
  n2 = Constant() {r=1.5}
  n3 = ScalarParameter() {group=AssetVariant, parameterName=isWisp?}
  n4 = Saturate(None=Power(
    Base=MaterialFunctionCall() {materialFunction=/SpecterGameplay/Assets/Sprite_Ghost/Materials/MaterialFunctions/MF_Sprite_Ghost_Opacity_A}.Color Mask
    Exp=LinearInterpolate(A=@n1, B=@n2, Alpha=@n3)
   ))
  Reroute(None=MaterialFunctionCall(Alpha Threshold=@n4) {materialFunction=&DitherTemporalAA}.Result)
  Reroute(None=@n4) {desc="isWisp = Translucent"}
  n5 = VertexColor()
  n6 = Constant() {r=4}
  n7 = Constant() {r=0.85}
  n8 = ScalarParameter() {group=AssetVariant, parameterName=isWisp?}
  n9 = Constant() {r=0}
  n10 = Constant() {r=0.04}
  n11 = ScalarParameter() {group=AssetVariant, parameterName=isWisp?}
  n12 = VertexColor()
  n13 = Multiply(
   A=OneMinus(None=MaterialFunctionCall() {materialFunction=&MF_Sprite_Ghost_Opacity_A}.Tex Out)
   B=Constant() {r=3}
  )
  n14 = Constant() {r=-0.2}
  n15 = Power(Base=Saturate(None=Add(A=@n13, B=@n14)), Exp=Constant() {desc=AlphaEdgeGlowPow, r=0.65})
  n16 = Add(
   A=Fresnel(ExponentIn=LinearInterpolate(A=@n6, B=@n7, Alpha=@n8), BaseReflectFractionIn=LinearInterpolate(A=@n9, B=@n10, Alpha=@n11)) {baseReflectFraction=0.0001, exponent=4}
   B=LinearInterpolate(A=Constant(), B=Multiply(A=@n12.R, B=@n15), Alpha=ScalarParameter() {group=AssetVariant, parameterName=isWisp?})
  ) {desc=isWisp-only}
  n17 = LinearInterpolate(
   A=VectorParameter() {parameterName=Color01, defaultValue=<rgba 0.0723 0 0.2051 1>}.RGB
   B=VectorParameter() {parameterName=Color02, defaultValue=<rgba 0.7011 0.552 0.8228 1>}.RGB
   Alpha=Saturate(None=Add(A=@n5.R, B=@n16))
  )
  n18 = MaterialFunctionCall(
   Value_Morning=ScalarParameter() {defaultValue=1, parameterName=TOD_Morning}
   Value_Day=ScalarParameter() {defaultValue=1, parameterName=TOD_Day}
   Value_Evening=ScalarParameter() {defaultValue=1, parameterName=TOD_Evening}
   Value_Night=ScalarParameter() {defaultValue=0.6, parameterName=TOD_Night}
  ) {materialFunction=&Scalar_Time_of_Day}
  Reroute(None=EyeAdaptationInverse(LightValueInput=Multiply(A=@n17, B=@n18.Result), AlphaInput=ScalarParameter() {defaultValue=1, parameterName=EyeAdaptionInverse}).EyeAdaptationInverse)
  ScalarParameter() {defaultValue=50, parameterName=Pixel Offset}
 ]
}
