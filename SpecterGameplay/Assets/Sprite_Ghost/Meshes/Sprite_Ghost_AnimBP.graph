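# Animation Blueprint for the Sprite_Ghost character, covering spawn, idle, walking, tac sprint, flying and jumping.
# A main state machine nests idle and traversal machines, with a randomized retriggerable timer alternating two look-around idle breaks.
# Landing plays as a dynamic additive over cached movement, with an optional tilt aim offset.
# Bool-switched blends layer this from root over the input pose, then feed DefaultSlot and inertialization.
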AnimBlueprint /SpecterGameplay/Assets/Sprite_Ghost/Meshes/Sprite_Ghost_AnimBP
target_skeleton=&/Sprite_Ghost_Skeleton
graphs [
 AnimGraph {nodes [
   n1 = AimOffset Player 'Sprite_Ghost_Tilt_AO'!(X:float=0, Y:float=0, BasePose:Pose Link Structure=Use cached pose 'PostLandingAdditive'(), Alpha:float=1) -> (Pose:Pose Link Structure) {blendSpace=/SpecterGameplay/Assets/Sprite_Ghost/Animation/Sprite_Ghost_Tilt_AO}
   n2 = Use cached pose 'PostLandingAdditive'()
   n3 = Blend Poses by bool(
    BlendPose_0=Use cached pose 'InputPose'()
    BlendPose_1=Layered blend per bone(
     BasePose=Use cached pose 'InputPose'()
     BlendPoses_0=Blend Poses by bool(BlendPose_0=@n1, BlendPose_1=@n2, BlendTime_0=0.2, BlendTime_1=1) {blendType=CubicInOut}
    ) {layerSetup [{branchFilters [{boneName=root}]}]}
   )
   Output Pose(Result=Inertialization!(Source:Pose Link Structure=Slot 'DefaultSlot'(Source=@n3)) -> (Pose:Pose Link Structure) {is=title AnimGraphNode_Inertialization})
   AnimGraphNode_StateMachine_0 = Anim States() {is=title AnimGraphNode_StateMachine, bAllowConduitEntryStates=true}
   Save cached pose 'MovementCache'(Pose=@AnimGraphNode_StateMachine_0) {cacheName=MovementCache}
   Save cached pose 'InputPose'(Pose=Input Pose()) {cacheName=InputPose}
   AnimGraphNode_StateMachine_1 = LandingAdditive_SM() {is=title AnimGraphNode_StateMachine, maxTransitionsPerFrame=1}
   Save cached pose 'PostLandingAdditive'(Pose=Apply Additive(Base=Use cached pose 'MovementCache'(), Additive=@AnimGraphNode_StateMachine_1)) {cacheName=PostLandingAdditive}
  ]}
 Anim States {
  entry=Conduit
  path=AnimGraph.AnimGraphNode_StateMachine_0.Anim States
  ids=id
  nodes [
   AnimStateConduitNode_0 = Conduit()
   AnimStateNode_1 = Idle() {stateEntered {guid=#1, notifyName=EnteredIdleState}}
   AnimStateNode_3 = Traversal() {
    is=title AnimStateNode
    stateEntered {guid=#2}
    stateFullyBlended {guid=#3}
    stateLeft {guid=#4}
   }
   AnimStateNode_4 = Jump Loop() {
    is=title AnimStateNode
    stateEntered {guid=#5}
    stateFullyBlended {guid=#6}
    stateLeft {guid=#7}
   }
   AnimStateNode_5 = Jump Start() {
    is=title AnimStateNode
    bAlwaysResetOnEntry=true
    stateEntered {guid=#8}
    stateFullyBlended {guid=#9}
    stateLeft {guid=#10}
   }
   AnimStateNode_8 = Spawn() {is=title AnimStateNode, bAlwaysResetOnEntry=true}
   AnimStateTransitionNode_0 = AnimStateTransitionNode() {bAllowInertializationForSelfTransitions=true}
   AnimStateTransitionNode_1 = AnimStateTransitionNode() {bAllowInertializationForSelfTransitions=true, crossfadeDuration=0.4}
   AnimStateTransitionNode_10 = AnimStateTransitionNode() {bAllowInertializationForSelfTransitions=true}
   AnimStateTransitionNode_19 = AnimStateTransitionNode() {bAllowInertializationForSelfTransitions=true, crossfadeDuration=0}
   AnimStateTransitionNode_2 = AnimStateTransitionNode() {bAllowInertializationForSelfTransitions=true, crossfadeDuration=0.1}
   AnimStateTransitionNode_20 = AnimStateTransitionNode() {bAllowInertializationForSelfTransitions=true, bAutomaticRuleBasedOnSequencePlayerInState=true, crossfadeDuration=0.4}
   AnimStateTransitionNode_21 = AnimStateTransitionNode() {bAllowInertializationForSelfTransitions=true}
   AnimStateTransitionNode_22 = AnimStateTransitionNode() {bAllowInertializationForSelfTransitions=true, crossfadeDuration=0.4}
   AnimStateTransitionNode_3 = AnimStateTransitionNode() {bAllowInertializationForSelfTransitions=true, crossfadeDuration=0.1}
   AnimStateTransitionNode_4 = AnimStateTransitionNode() {
    bAllowInertializationForSelfTransitions=true
    bAutomaticRuleBasedOnSequencePlayerInState=true
    crossfadeDuration=0.35
    logicType=TLT_Inertialization
   }
   AnimStateTransitionNode_5 = AnimStateTransitionNode() {bAllowInertializationForSelfTransitions=true}
   AnimStateTransitionNode_6 = AnimStateTransitionNode() {bAllowInertializationForSelfTransitions=true}
   AnimStateTransitionNode_8 = AnimStateTransitionNode() {bAllowInertializationForSelfTransitions=true, crossfadeDuration=0.1}
  ]
  links [
   {
    dst=Idle
    rule=AnimGraph.AnimGraphNode_StateMachine_0.Anim States.AnimStateTransitionNode_22.Transition
    src=Conduit
    via=AnimStateTransitionNode_22
   }
   {
    dst=Spawn
    rule=AnimGraph.AnimGraphNode_StateMachine_0.Anim States.AnimStateTransitionNode_19.Transition
    src=Conduit
    via=AnimStateTransitionNode_19
   }
   {
    dst=Jump Loop
    rule=AnimGraph.AnimGraphNode_StateMachine_0.Anim States.AnimStateTransitionNode_21.Transition
    src=Idle
    via=AnimStateTransitionNode_21
   }
   {
    dst=Jump Start
    rule=AnimGraph.AnimGraphNode_StateMachine_0.Anim States.AnimStateTransitionNode_2.Transition
    src=Idle
    via=AnimStateTransitionNode_2
   }
   {
    dst=Traversal
    rule=AnimGraph.AnimGraphNode_StateMachine_0.Anim States.AnimStateTransitionNode_0.Transition
    src=Idle
    via=AnimStateTransitionNode_0
   }
   {
    dst=Idle
    rule=AnimGraph.AnimGraphNode_StateMachine_0.Anim States.AnimStateTransitionNode_10.Transition
    src=Jump Loop
    via=AnimStateTransitionNode_10
   }
   {
    dst=Idle
    rule=AnimGraph.AnimGraphNode_StateMachine_0.Anim States.AnimStateTransitionNode_5.Transition
    src=Jump Loop
    via=AnimStateTransitionNode_5
   }
   {
    dst=Idle
    rule=AnimGraph.AnimGraphNode_StateMachine_0.Anim States.AnimStateTransitionNode_6.Transition
    src=Jump Loop
    via=AnimStateTransitionNode_6
   }
   {
    dst=Jump Start
    rule=AnimGraph.AnimGraphNode_StateMachine_0.Anim States.AnimStateTransitionNode_8.Transition
    src=Jump Loop
    via=AnimStateTransitionNode_8
   }
   {
    dst=Jump Loop
    rule=AnimGraph.AnimGraphNode_StateMachine_0.Anim States.AnimStateTransitionNode_4.Transition
    src=Jump Start
    via=AnimStateTransitionNode_4
   }
   {
    dst=Idle
    rule=AnimGraph.AnimGraphNode_StateMachine_0.Anim States.AnimStateTransitionNode_20.Transition
    src=Spawn
    via=AnimStateTransitionNode_20
   }
   {
    dst=Idle
    rule=AnimGraph.AnimGraphNode_StateMachine_0.Anim States.AnimStateTransitionNode_1.Transition
    src=Traversal
    via=AnimStateTransitionNode_1
   }
   {
    dst=Jump Start
    rule=AnimGraph.AnimGraphNode_StateMachine_0.Anim States.AnimStateTransitionNode_3.Transition
    src=Traversal
    via=AnimStateTransitionNode_3
   }
  ]
 }
 Conduit {path=AnimGraph.AnimGraphNode_StateMachine_0.Anim States.AnimStateConduitNode_0.Conduit, nodes [Result(bCanEnterTransition=true)]}
 Idle {
  path=AnimGraph.AnimGraphNode_StateMachine_0.Anim States.AnimStateNode_1.Idle
  nodes [
   AnimGraphNode_StateMachine_0 = Idle_SM() {is=title AnimGraphNode_StateMachine}
   Output Animation Pose(Result=@AnimGraphNode_StateMachine_0) {
    node {stateEntryFunction {functionName=ResetIdle}}
    stateEntryFunction {bSelfContext=true, memberGuid=#11, memberName=ResetIdle}
   }
  ]
 }
 Idle_SM {
  entry=Idle
  path=AnimGraph.AnimGraphNode_StateMachine_0.Anim States.AnimStateNode_1.Idle.AnimGraphNode_StateMachine_0.Idle_SM
  ids=id
  nodes [
   AnimStateNode_0 = Idle()
   AnimStateNode_6 = Idle_LookBack() {is=title AnimStateNode, bAlwaysResetOnEntry=true}
   AnimStateNode_7 = Idle_LookAround() {is=title AnimStateNode, bAlwaysResetOnEntry=true}
   AnimStateTransitionNode_2 = AnimStateTransitionNode() {bAllowInertializationForSelfTransitions=true, crossfadeDuration=0.4}
   AnimStateTransitionNode_3 = AnimStateTransitionNode() {bAllowInertializationForSelfTransitions=true, bAutomaticRuleBasedOnSequencePlayerInState=true, crossfadeDuration=0.4}
   AnimStateTransitionNode_4 = AnimStateTransitionNode() {bAllowInertializationForSelfTransitions=true, crossfadeDuration=0.4}
   AnimStateTransitionNode_5 = AnimStateTransitionNode() {bAllowInertializationForSelfTransitions=true, bAutomaticRuleBasedOnSequencePlayerInState=true, crossfadeDuration=0.4}
  ]
  links [
   {
    dst=Idle_LookAround
    rule=AnimGraph.AnimGraphNode_StateMachine_0.Anim States.AnimStateNode_1.Idle.AnimGraphNode_StateMachine_0.Idle_SM.AnimStateTransitionNode_4.Transition
    src=Idle
    via=AnimStateTransitionNode_4
   }
   {
    dst=Idle_LookBack
    rule=AnimGraph.AnimGraphNode_StateMachine_0.Anim States.AnimStateNode_1.Idle.AnimGraphNode_StateMachine_0.Idle_SM.AnimStateTransitionNode_2.Transition
    src=Idle
    via=AnimStateTransitionNode_2
   }
   {
    dst=Idle
    rule=AnimGraph.AnimGraphNode_StateMachine_0.Anim States.AnimStateNode_1.Idle.AnimGraphNode_StateMachine_0.Idle_SM.AnimStateTransitionNode_5.Transition
    src=Idle_LookAround
    via=AnimStateTransitionNode_5
   }
   {
    dst=Idle
    rule=AnimGraph.AnimGraphNode_StateMachine_0.Anim States.AnimStateNode_1.Idle.AnimGraphNode_StateMachine_0.Idle_SM.AnimStateTransitionNode_3.Transition
    src=Idle_LookBack
    via=AnimStateTransitionNode_3
   }
  ]
 }
 Idle {
  path=AnimGraph.AnimGraphNode_StateMachine_0.Anim States.AnimStateNode_1.Idle.AnimGraphNode_StateMachine_0.Idle_SM.AnimStateNode_0.Idle
  nodes [Output Animation Pose(Result=Sequence Player 'Sprite_Ghost_Idle'!(bLoopAnimation:bool=True) -> (Pose) {sequence=/SpecterGameplay/Assets/Sprite_Ghost/Animation/Sprite_Ghost_Idle}) {
    node {stateEntryFunction {functionName=ResetIdle}}
    stateEntryFunction {bSelfContext=true, memberGuid=#11, memberName=ResetIdle}
   }]
 }
 Idle_LookBack {
  path=AnimGraph.AnimGraphNode_StateMachine_0.Anim States.AnimStateNode_1.Idle.AnimGraphNode_StateMachine_0.Idle_SM.AnimStateNode_6.Idle_LookBack
  nodes [Output Animation Pose(Result=Sequence Player 'Sprite_Ghost_IdleBreak_LookBack'() {bLoopAnimation=false, sequence=/SpecterGameplay/Assets/Sprite_Ghost/Animation/Sprite_Ghost_IdleBreak_LookBack})]
 }
 Idle_LookAround {
  path=AnimGraph.AnimGraphNode_StateMachine_0.Anim States.AnimStateNode_1.Idle.AnimGraphNode_StateMachine_0.Idle_SM.AnimStateNode_7.Idle_LookAround
  nodes [Output Animation Pose(Result=Sequence Player 'Sprite_Ghost_IdleBreak_LookAround'() {bLoopAnimation=false, sequence=/SpecterGameplay/Assets/Sprite_Ghost/Animation/Sprite_Ghost_IdleBreak_LookAround})]
 }
 Transition {
  path=AnimGraph.AnimGraphNode_StateMachine_0.Anim States.AnimStateNode_1.Idle.AnimGraphNode_StateMachine_0.Idle_SM.AnimStateTransitionNode_2.Transition
  nodes [Result(bCanEnterTransition=Get CanEnterIdleBreakout1() -> (CanEnterIdleBreakout1:bool))]
 }
 Transition {
  path=AnimGraph.AnimGraphNode_StateMachine_0.Anim States.AnimStateNode_1.Idle.AnimGraphNode_StateMachine_0.Idle_SM.AnimStateTransitionNode_3.Transition
  nodes [Result()]
 }
 Transition {
  path=AnimGraph.AnimGraphNode_StateMachine_0.Anim States.AnimStateNode_1.Idle.AnimGraphNode_StateMachine_0.Idle_SM.AnimStateTransitionNode_4.Transition
  nodes [Result(bCanEnterTransition=Get CanEnterIdleBreakout2() -> (CanEnterIdleBreakout2:bool))]
 }
 Transition {
  path=AnimGraph.AnimGraphNode_StateMachine_0.Anim States.AnimStateNode_1.Idle.AnimGraphNode_StateMachine_0.Idle_SM.AnimStateTransitionNode_5.Transition
  nodes [Result()]
 }
 Traversal {
  path=AnimGraph.AnimGraphNode_StateMachine_0.Anim States.AnimStateNode_3.Traversal
  nodes [
   AnimGraphNode_StateMachine_0 = Traversal_SM() {is=title AnimGraphNode_StateMachine, bAllowConduitEntryStates=true}
   Output Animation Pose(Result=@AnimGraphNode_StateMachine_0)
  ]
 }
 Traversal_SM {
  entry=Conduit
  path=AnimGraph.AnimGraphNode_StateMachine_0.Anim States.AnimStateNode_3.Traversal.AnimGraphNode_StateMachine_0.Traversal_SM
  ids=id
  nodes [
   AnimStateConduitNode_0 = Conduit()
   AnimStateNode_0 = Walking() {is=title AnimStateNode}
   AnimStateNode_1 = TacSprint() {is=title AnimStateNode}
   AnimStateNode_2 = Flying() {is=title AnimStateNode}
   AnimStateTransitionNode_17 = AnimStateTransitionNode() {bAllowInertializationForSelfTransitions=true}
   AnimStateTransitionNode_18 = AnimStateTransitionNode() {bAllowInertializationForSelfTransitions=true}
   AnimStateTransitionNode_19 = AnimStateTransitionNode() {bAllowInertializationForSelfTransitions=true, crossfadeDuration=1}
   AnimStateTransitionNode_20 = AnimStateTransitionNode() {bAllowInertializationForSelfTransitions=true}
   AnimStateTransitionNode_21 = AnimStateTransitionNode() {bAllowInertializationForSelfTransitions=true}
   AnimStateTransitionNode_6 = AnimStateTransitionNode() {bAllowInertializationForSelfTransitions=true, crossfadeDuration=0.4}
  ]
  links [
   {
    dst=Flying
    rule=AnimGraph.AnimGraphNode_StateMachine_0.Anim States.AnimStateNode_3.Traversal.AnimGraphNode_StateMachine_0.Traversal_SM.AnimStateTransitionNode_18.Transition
    src=Conduit
    via=AnimStateTransitionNode_18
   }
   {
    dst=TacSprint
    rule=AnimGraph.AnimGraphNode_StateMachine_0.Anim States.AnimStateNode_3.Traversal.AnimGraphNode_StateMachine_0.Traversal_SM.AnimStateTransitionNode_21.Transition
    src=Conduit
    via=AnimStateTransitionNode_21
   }
   {
    dst=Walking
    rule=AnimGraph.AnimGraphNode_StateMachine_0.Anim States.AnimStateNode_3.Traversal.AnimGraphNode_StateMachine_0.Traversal_SM.AnimStateTransitionNode_17.Transition
    src=Conduit
    via=AnimStateTransitionNode_17
   }
   {
    dst=TacSprint
    rule=AnimGraph.AnimGraphNode_StateMachine_0.Anim States.AnimStateNode_3.Traversal.AnimGraphNode_StateMachine_0.Traversal_SM.AnimStateTransitionNode_19.Transition
    src=Flying
    via=AnimStateTransitionNode_19
   }
   {
    dst=Conduit
    rule=AnimGraph.AnimGraphNode_StateMachine_0.Anim States.AnimStateNode_3.Traversal.AnimGraphNode_StateMachine_0.Traversal_SM.AnimStateTransitionNode_20.Transition
    src=TacSprint
    via=AnimStateTransitionNode_20
   }
   {
    dst=TacSprint
    rule=AnimGraph.AnimGraphNode_StateMachine_0.Anim States.AnimStateNode_3.Traversal.AnimGraphNode_StateMachine_0.Traversal_SM.AnimStateTransitionNode_6.Transition
    src=Walking
    via=AnimStateTransitionNode_6
   }
  ]
 }
 Conduit {
  path=AnimGraph.AnimGraphNode_StateMachine_0.Anim States.AnimStateNode_3.Traversal.AnimGraphNode_StateMachine_0.Traversal_SM.AnimStateConduitNode_0.Conduit
  nodes [Result(bCanEnterTransition=true)]
 }
 Walking {
  path=AnimGraph.AnimGraphNode_StateMachine_0.Anim States.AnimStateNode_3.Traversal.AnimGraphNode_StateMachine_0.Traversal_SM.AnimStateNode_0.Walking
  nodes [Output Animation Pose(Result=Blendspace Player 'Sprite_Ghost_Loco_Grounded_BS1'(X=0, Y=0) {blendSpace=/SpecterGameplay/Assets/Sprite_Ghost/Animation/Sprite_Ghost_Loco_Grounded_BS1}) {
    node {stateEntryFunction {functionName=OnStateEntry}}
    stateEntryFunction {bSelfContext=true, memberGuid=#12, memberName=ResetWasFlying}
   }]
 }
 TacSprint {
  path=AnimGraph.AnimGraphNode_StateMachine_0.Anim States.AnimStateNode_3.Traversal.AnimGraphNode_StateMachine_0.Traversal_SM.AnimStateNode_1.TacSprint
  nodes [Output Animation Pose(Result=Sequence Player 'Sprite_Ghost_Sprint'() {sequence=/SpecterGameplay/Assets/Sprite_Ghost/Animation/Sprite_Ghost_Sprint})]
 }
 Flying {
  path=AnimGraph.AnimGraphNode_StateMachine_0.Anim States.AnimStateNode_3.Traversal.AnimGraphNode_StateMachine_0.Traversal_SM.AnimStateNode_2.Flying
  nodes [Output Animation Pose(Result=Blendspace Player 'Sprite_Ghost_Flying_BS'(X=0, Y=0) {blendSpace=/SpecterGameplay/Assets/Sprite_Ghost/Animation/Sprite_Ghost_Flying_BS}) {
    node {stateEntryFunction {functionName=OnStateEntry}}
    stateEntryFunction {bSelfContext=true, memberGuid=#13, memberName=OnFlyingStateEntry}
   }]
 }
 Transition {
  path=AnimGraph.AnimGraphNode_StateMachine_0.Anim States.AnimStateNode_3.Traversal.AnimGraphNode_StateMachine_0.Traversal_SM.AnimStateTransitionNode_17.Transition
  nodes [Result(bCanEnterTransition=Get ABP_IsGrounded() -> (ABP_IsGrounded:bool))]
 }
 Transition {
  path=AnimGraph.AnimGraphNode_StateMachine_0.Anim States.AnimStateNode_3.Traversal.AnimGraphNode_StateMachine_0.Traversal_SM.AnimStateTransitionNode_18.Transition
  nodes [Result(bCanEnterTransition=Get bIsFlying() -> (bIsFlying:bool))]
 }
 Transition {
  path=AnimGraph.AnimGraphNode_StateMachine_0.Anim States.AnimStateNode_3.Traversal.AnimGraphNode_StateMachine_0.Traversal_SM.AnimStateTransitionNode_19.Transition
  nodes [Result(bCanEnterTransition=Get bIsTacSprinting() -> (bIsTacSprinting:bool))]
 }
 Transition {
  path=AnimGraph.AnimGraphNode_StateMachine_0.Anim States.AnimStateNode_3.Traversal.AnimGraphNode_StateMachine_0.Traversal_SM.AnimStateTransitionNode_20.Transition
  nodes [Result(bCanEnterTransition=NOT Boolean(A=Get bIsTacSprinting()))]
 }
 Transition {
  path=AnimGraph.AnimGraphNode_StateMachine_0.Anim States.AnimStateNode_3.Traversal.AnimGraphNode_StateMachine_0.Traversal_SM.AnimStateTransitionNode_21.Transition
  nodes [Result(bCanEnterTransition=Get bIsTacSprinting())]
 }
 Transition {
  path=AnimGraph.AnimGraphNode_StateMachine_0.Anim States.AnimStateNode_3.Traversal.AnimGraphNode_StateMachine_0.Traversal_SM.AnimStateTransitionNode_6.Transition
  nodes [Result(bCanEnterTransition=Get bIsTacSprinting())]
 }
 Jump Loop {
  path=AnimGraph.AnimGraphNode_StateMachine_0.Anim States.AnimStateNode_4.Jump Loop
  nodes [Output Animation Pose(Result=Sequence Player 'Sprite_Ghost_Jump_Loop'!(bLoopAnimation:bool=True) -> (Pose) {sequence=/SpecterGameplay/Assets/Sprite_Ghost/Animation/Sprite_Ghost_Jump_Loop}) {
    node {stateEntryFunction {functionName=ResetWasFlying}}
    stateEntryFunction {bSelfContext=true, memberGuid=#12, memberName=ResetWasFlying}
   }]
 }
 Jump Start {
  path=AnimGraph.AnimGraphNode_StateMachine_0.Anim States.AnimStateNode_5.Jump Start
  nodes [Output Animation Pose(Result=Sequence Player 'Sprite_Ghost_Jump_Start'() {bLoopAnimation=false, sequence=/SpecterGameplay/Assets/Sprite_Ghost/Animation/Sprite_Ghost_Jump_Start}) {
    node {stateEntryFunction {functionName=ResetWasFlying}}
    stateEntryFunction {bSelfContext=true, memberGuid=#12, memberName=ResetWasFlying}
   }]
 }
 Spawn {
  path=AnimGraph.AnimGraphNode_StateMachine_0.Anim States.AnimStateNode_8.Spawn
  nodes [Output Animation Pose(Result=Sequence Player 'Sprite_Ghost_Spawn'() {bLoopAnimation=false, sequence=/SpecterGameplay/Assets/Sprite_Ghost/Animation/Sprite_Ghost_Spawn}) {
    node {stateExitFunction {functionName=OnStateExit}}
    stateExitFunction {bSelfContext=true, memberGuid=#14, memberName=ResetSpawn}
   }]
 }
 Transition {
  path=AnimGraph.AnimGraphNode_StateMachine_0.Anim States.AnimStateTransitionNode_0.Transition
  nodes [Result(bCanEnterTransition=Get ABP_IsMoving() -> (ABP_IsMoving:bool))]
 }
 Transition {
  path=AnimGraph.AnimGraphNode_StateMachine_0.Anim States.AnimStateTransitionNode_1.Transition
  nodes [Result(bCanEnterTransition=NOT Boolean(A=Get ABP_IsMoving()))]
 }
 Transition {
  path=AnimGraph.AnimGraphNode_StateMachine_0.Anim States.AnimStateTransitionNode_10.Transition
  nodes [Result(bCanEnterTransition=Get bIsFlying())]
 }
 Transition {
  path=AnimGraph.AnimGraphNode_StateMachine_0.Anim States.AnimStateTransitionNode_19.Transition
  nodes [Result(bCanEnterTransition=Get Spawning() -> (Spawning:bool))]
 }
 Transition {
  path=AnimGraph.AnimGraphNode_StateMachine_0.Anim States.AnimStateTransitionNode_2.Transition
  nodes [Result(bCanEnterTransition=Get ABP_JumpTriggered() -> (ABP_JumpTriggered:bool))]
 }
 Transition {path=AnimGraph.AnimGraphNode_StateMachine_0.Anim States.AnimStateTransitionNode_20.Transition, nodes [Result()]}
 Transition {
  path=AnimGraph.AnimGraphNode_StateMachine_0.Anim States.AnimStateTransitionNode_21.Transition
  nodes [Result(bCanEnterTransition=Get bIsFalling())]
 }
 Transition {
  path=AnimGraph.AnimGraphNode_StateMachine_0.Anim States.AnimStateTransitionNode_22.Transition
  nodes [Result(bCanEnterTransition=NOT Boolean(A=Get Spawning()))]
 }
 Transition {
  path=AnimGraph.AnimGraphNode_StateMachine_0.Anim States.AnimStateTransitionNode_3.Transition
  nodes [Result(bCanEnterTransition=Get ABP_JumpTriggered())]
 }
 Transition {path=AnimGraph.AnimGraphNode_StateMachine_0.Anim States.AnimStateTransitionNode_4.Transition, nodes [Result()]}
 Transition {
  path=AnimGraph.AnimGraphNode_StateMachine_0.Anim States.AnimStateTransitionNode_5.Transition
  nodes [Result(bCanEnterTransition=Get ABP_IsGrounded())]
 }
 Transition {
  path=AnimGraph.AnimGraphNode_StateMachine_0.Anim States.AnimStateTransitionNode_6.Transition
  nodes [Result(bCanEnterTransition=Get bJustLanded() -> (bJustLanded:bool))]
 }
 Transition {
  path=AnimGraph.AnimGraphNode_StateMachine_0.Anim States.AnimStateTransitionNode_8.Transition
  nodes [Result(bCanEnterTransition=Get ABP_JumpTriggered())]
 }
 LandingAdditive_SM {
  entry=Locomotion
  path=AnimGraph.AnimGraphNode_StateMachine_1.LandingAdditive_SM
  ids=id
  nodes [
   AnimStateNode_1 = LandingAdditive() {is=title AnimStateNode}
   AnimStateNode_2 = Locomotion() {
    is=title AnimStateNode
    stateEntered {guid=#15}
    stateFullyBlended {guid=#16}
    stateLeft {guid=#17}
   }
   AnimStateTransitionNode_8 = AnimStateTransitionNode() {bAllowInertializationForSelfTransitions=true}
   AnimStateTransitionNode_9 = AnimStateTransitionNode() {
    automaticRuleTriggerTime=0.6
    bAllowInertializationForSelfTransitions=true
    bAutomaticRuleBasedOnSequencePlayerInState=true
    crossfadeDuration=0.5
   }
  ]
  links [
   {
    dst=Locomotion
    rule=AnimGraph.AnimGraphNode_StateMachine_1.LandingAdditive_SM.AnimStateTransitionNode_9.Transition
    src=LandingAdditive
    via=AnimStateTransitionNode_9
   }
   {
    dst=LandingAdditive
    rule=AnimGraph.AnimGraphNode_StateMachine_1.LandingAdditive_SM.AnimStateTransitionNode_8.Transition
    src=Locomotion
    via=AnimStateTransitionNode_8
   }
  ]
 }
 LandingAdditive {
  path=AnimGraph.AnimGraphNode_StateMachine_1.LandingAdditive_SM.AnimStateNode_1.LandingAdditive
  nodes [Output Animation Pose(Result=Make Dynamic Additive!(
     Base:Pose Link Structure=Use cached pose 'MovementCache'()
     Additive:Pose Link Structure=Sequence Player 'Sprite_Ghost_Jump_End'() {bLoopAnimation=false, sequence=/SpecterGameplay/Assets/Sprite_Ghost/Animation/Sprite_Ghost_Jump_End}
    ) -> (Pose:Pose Link Structure) {is=title AnimGraphNode_MakeDynamicAdditive})]
 }
 Locomotion {
  path=AnimGraph.AnimGraphNode_StateMachine_1.LandingAdditive_SM.AnimStateNode_2.Locomotion
  nodes [Output Animation Pose(Result=Additive Identity Pose())]
 }
 Transition {
  path=AnimGraph.AnimGraphNode_StateMachine_1.LandingAdditive_SM.AnimStateTransitionNode_8.Transition
  nodes [Result(bCanEnterTransition=Get bJustLanded())]
 }
 Transition {path=AnimGraph.AnimGraphNode_StateMachine_1.LandingAdditive_SM.AnimStateTransitionNode_9.Transition, nodes [Result()]}
 EventGraph {nodes [
   n1 = AnimNotify_EnteredIdleState() {is=title K2Node_Event}
   RetriggerableDelay(execute=[@n1.then, @n4.then, @n5.then, @n2.then, @n6.then_0], Duration=RandomFloatInRange(Min=6, Max=12))
   > n2 = Branch(Condition=OR Boolean!(A=Get CanEnterIdleBreakout2(), B=Get CanEnterIdleBreakout1(), C:bool=Get ABP_IsMoving()) -> (ReturnValue))
   n3 = Flip Flop!(None:Exec=@n2.else) -> (A:Exec, B:Exec, IsA:bool=~) {is=title K2Node_MacroInstance}
   > n4 = Set CanEnterIdleBreakout1(CanEnterIdleBreakout1:bool=true) -> (Output_Get:bool)
   n5 = Set CanEnterIdleBreakout2(execute=@n3.B, CanEnterIdleBreakout2:bool=true) -> (Output_Get:bool)
   Event BlueprintInitializeAnimation()
   > n6 = Sequence()
   Set Spawning(execute=@n6.then_1, Spawning:bool=true) -> (Output_Get:bool)
  ]}
 OnFlyingStateEntry {nodes [
   OnFlyingStateEntry!() -> (then, Context:"Anim Update Context Structure (by ref)", Node:"Anim Node Reference Structure (by ref)") {is=title K2Node_FunctionEntry}
   > Set bWasFlying(bWasFlying:bool=true) -> (Output_Get:bool)
   > Return Node()
  ]}
 ResetIdle {nodes [
   ResetIdle!() -> (then, Context:"Anim Update Context Structure (by ref)", Node:"Anim Node Reference Structure (by ref)") {is=title K2Node_FunctionEntry}
   > Set CanEnterIdleBreakout1(CanEnterIdleBreakout1=false)
   > Set CanEnterIdleBreakout2(CanEnterIdleBreakout2=false)
   > n1 = Branch(Condition=Get ABP_IsGrounded())
   > n2 = Set bWasFlying(bWasFlying=false)
   Return Node(execute=[@n2.then, @n1.else])
  ]}
 ResetSpawn {nodes [
   ResetSpawn!() -> (then, Context:"Anim Update Context Structure (by ref)", Node:"Anim Node Reference Structure (by ref)") {is=title K2Node_FunctionEntry}
   > Set Spawning(Spawning=false)
   > Return Node()
  ]}
 ResetWasFlying {nodes [
   ResetWasFlying!() -> (then, Context:"Anim Update Context Structure (by ref)", Node:"Anim Node Reference Structure (by ref)") {is=title K2Node_FunctionEntry}
   > Set bWasFlying(bWasFlying=false)
   > Return Node()
  ]}
]
variables [
 CanEnterIdleBreakout1 {type=boolean}
 CanEnterIdleBreakout2 {type=boolean}
 DoIdleBreak {type=boolean}
 Spawning {type=boolean, default=true}
]
